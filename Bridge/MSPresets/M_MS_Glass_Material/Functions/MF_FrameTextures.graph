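# MF_FrameTextures is a library-exposed material function in the Megascans glass material preset that supplies the frame's texture set.
# Five texture parameters in the Frame Textures group feed Albedo, Normal, Metallic, Roughness and Opacity outputs, with the last three taking the red channel.
# Albedo and Normal use the shared wrap sampler.
# Only Albedo has a default texture assigned, a white one.
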MaterialFunction /Bridge/MSPresets/M_MS_Glass_Material/Functions/MF_FrameTextures
graph {
 properties {bExposeToLibrary=true}
 nodes [
  n1 = TextureSampleParameter2D() {
   group=Frame Textures
   parameterName=Frame Albedo
   samplerSource=SSM_Wrap_WorldGroupSettings
   samplerType=SAMPLERTYPE_LinearColor
   texture=/Bridge/MSPresets/M_MS_Glass_Material/Textures/White
  }
  FunctionOutput(None=@n1.RGB) {outputName=Albedo}
  FunctionOutput(None=TextureSampleParameter2D() {group=Frame Textures, parameterName=Frame Metallic, texture=null}.R) {outputName=Metallic, sortPriority=1}
  FunctionOutput(None=TextureSampleParameter2D() {group=Frame Textures, parameterName=Frame Opacity, samplerType=SAMPLERTYPE_LinearColor, texture=null}.R) {outputName=Opacity, sortPriority=4}
  n2 = TextureSampleParameter2D() {group=Frame Textures, parameterName=Frame Normal, samplerSource=SSM_Wrap_WorldGroupSettings, samplerType=SAMPLERTYPE_Normal, texture=null}
  FunctionOutput(None=@n2.RGB) {outputName=Normal, sortPriority=5}
  FunctionOutput(None=TextureSampleParameter2D() {group=Frame Textures, parameterName=Frame Roughness, texture=null}.R) {outputName=Roughness, sortPriority=1}
 ]
}
