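# Material function for the glass material preset that samples an imperfection map to drive roughness variation.
# UVs are tiled, offset and spun by CustomRotator around the tiled center, using only the X tiling and offset values for both axes.
# The map's red channel is output raw as Roughness, and also scaled by Imperfection Roughness through a lerp, then raised to a contrast power for Adjust Roughness.
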MaterialFunction /Bridge/MSPresets/M_MS_Glass_Material/Functions/MF_Imperfection
graph {
 properties {bExposeToLibrary=true}
 nodes [
  n1 = ScalarParameter() {defaultValue=1, group=02 - Glass Imperfection, parameterName=Imperfection Roughness, sortPriority=6}
  n2 = VectorParameter() {
   group=06 - Emissive
   parameterName=Imperfection Tiling/Offset
   sortPriority=3
   channelNames {a=Offset Y, b=Offset X, g=Tiling Y, r=Tiling X}
   defaultValue=<rgba 1 1 0 0>
  }
  n3 = AppendVector(A=@n2.Tiling X, B=@n2.Tiling X)
  n4 = Multiply(A=TextureCoordinate(), B=@n3)
  n5 = AppendVector(A=@n2.Offset X, B=@n2.Offset X)
  n6 = Divide(A=@n3)
  n7 = TextureSampleParameter2D(UVs=MaterialFunctionCall(
    UVs=Add(A=@n4, B=@n5)
    Rotation Center=Add(A=@n6, B=@n5)
    "Rotation Angle (0-1)"=ScalarParameter() {group=02 - Glass Imperfection, parameterName=Imperfection Rotation, sortPriority=5}
   ) {materialFunction=&CustomRotator}.Rotated Values) {
   group=02 - Glass Imperfection
   parameterName=Imperfection Map
   samplerSource=SSM_Wrap_WorldGroupSettings
   texture=/Bridge/MSPresets/M_MS_Glass_Material/Textures/Black
  }
  FunctionOutput(None=Power(
    Base=LinearInterpolate(B=@n1, Alpha=@n7.R)
    Exp=ScalarParameter() {defaultValue=1, group=02 - Glass Imperfection, parameterName=Imperfection Contrast, sortPriority=7}
   )) {outputName=Adjust Roughness, sortPriority=1}
  FunctionOutput(None=@n7.R) {outputName=Roughness}
 ]
}
