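# Material function in the Megascans glass material presets that computes a metallic value for translucent surfaces.
# A static switch, Use Metallic Texture, picks either the Base Metallic scalar or a texture-driven value.
# The textured path samples the Metallic Texture input with the Tiling input as UVs, then uses its red channel to lerp from Min Metallic to Base Metallic.
# The texture input falls back to a black placeholder.
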MaterialFunction /Bridge/MSPresets/M_MS_Glass_Material/Functions/MF_TranslucencyMetallic
graph {nodes [
  n1 = ScalarParameter() {group=04 - Metallic, parameterName=Min Metallic, sortPriority=2}
  n2 = ScalarParameter() {defaultValue=0.01, group=04 - Metallic, parameterName=Base Metallic, sliderMax=1, sortPriority=1}
  n3 = TextureSample(
   UVs=FunctionInput(Preview=Constant2Vector() {g=1, r=1}) {bUsePreviewValueAsDefault=true, inputName=Tiling, inputType=FunctionInput_Vector2}
   Tex=FunctionInput(Preview=TextureObject() {texture=/Bridge/MSPresets/MSTextures/BlackPlaceholder_Color}) {bUsePreviewValueAsDefault=true, inputName=Metallic Texture, inputType=FunctionInput_Texture2D}
  )
  FunctionOutput(None=StaticSwitch(
    True=LinearInterpolate(A=@n1, B=@n2, Alpha=@n3.R)
    False=@n2
    Value=StaticBoolParameter() {group=04 - Metallic, parameterName=Use Metallic Texture}
   ))
 ]}
