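# Material function in the Megascans glass material preset that adjusts the strength of a normal map, named for window use.
# It samples a texture input, which defaults to the engine's flat normal, and multiplies the red and green channels by the Normal Strength scalar parameter.
# The untouched blue channel is appended back as the output, with no renormalization.
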MaterialFunction /Bridge/MSPresets/M_MS_Glass_Material/Functions/MF_WindowNormalStrength
graph {nodes [
  n1 = TextureSample(Tex=FunctionInput(Preview=TextureObject() {samplerType=SAMPLERTYPE_Normal, texture=/Engine/EngineMaterials/FlatNormal}) {bUsePreviewValueAsDefault=true, inputName=Normal, inputType=FunctionInput_Texture2D})
  n2 = ComponentMask(None=@n1.RGB) {g=true, r=true}
  n3 = ScalarParameter() {defaultValue=1, group=04 - Normals, parameterName=Normal Strength}
  FunctionOutput(None=AppendVector(A=Multiply(A=@n2, B=@n3), B=ComponentMask(None=@n1.RGB) {b=true}))
  Reroute()
 ]}
