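# Material function for the Megascans glass material that blends a textured frame into glass, using the Opacity map's red channel as mask.
# The mask lerps glass color, roughness, specular, refraction and emissive inputs toward frame maps or constants.
# Opacity combines base value, roughness-driven imperfection and mask; helper functions supply metallic and normal.
# A static switch selects plain glass, keeping those three shared, and outputs material attributes plus opacity.
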MaterialFunction /Bridge/MSPresets/M_MS_Glass_Material/Functions/MF_FrameBlend
graph {nodes [
  n1 = FunctionInput() {inputName=Glass Color}
  n2 = TextureObjectParameter() {group=07 - Texture Maps, parameterName=Albedo, texture=/Bridge/MSPresets/MSTextures/Placeholder}
  n3 = TextureObjectParameter() {group=07 - Texture Maps, parameterName=Opacity, sortPriority=4, texture=/Bridge/MSPresets/MSTextures/BlackPlaceholder_Color}
  n4 = TextureSample(Tex=@n3) {samplerType=SAMPLERTYPE_LinearColor}
  n5 = TextureObjectParameter() {
   group=07 - Texture Maps
   parameterName=Metallic
   samplerType=SAMPLERTYPE_LinearColor
   sortPriority=1
   texture=/Bridge/MSPresets/MSTextures/WhitePlaceholder
  }
  n6 = MaterialFunctionCall!(Tiling, Metallic Texture=@n5) {materialFunction=&/MF_TranslucencyMetallic}
  n7 = FunctionInput() {inputName=Specular, sortPriority=5}
  n8 = FunctionInput() {inputName=Roughness, inputType=FunctionInput_Scalar, sortPriority=2}
  n9 = TextureObjectParameter() {
   group=07 - Texture Maps
   parameterName=Roughness
   samplerType=SAMPLERTYPE_LinearColor
   sortPriority=2
   texture=&WhitePlaceholder
  }
  n10 = TextureObjectParameter() {
   group=07 - Texture Maps
   parameterName=Normal
   samplerType=SAMPLERTYPE_Normal
   sortPriority=5
   texture=/Bridge/MSPresets/MSTextures/Placeholder_Normal
  }
  n11 = MaterialFunctionCall!(Normal=@n10) {materialFunction=&/MF_WindowNormalStrength}
  n12 = FunctionInput() {inputName=Refraction, inputType=FunctionInput_Scalar, sortPriority=6}
  n13 = ScalarParameter() {defaultValue=0.01, group=01 - Opacity, parameterName=Base Opacity, sliderMax=1, sortPriority=1}
  n14 = LinearInterpolate(
   A=LinearInterpolate(A=@n13, B=Add(A=@n13, B=@n8), Alpha=ScalarParameter() {group=01 - Opacity, parameterName=Window Imperfection Opacity, sortPriority=1})
   B=@n4.R
   Alpha=@n4.R
  )
  n15 = FunctionInput() {inputName=Emissive, sortPriority=4}
  n16 = SetMaterialAttributes!(
   MaterialAttributes
   BaseColor=LinearInterpolate(A=@n1, B=TextureSample(Tex=@n2).RGB, Alpha=@n4.R)
   Metallic=@n6.Result
   Specular=LinearInterpolate(A=@n7, Alpha=@n4.R) {constB=0.5}
   Roughness=LinearInterpolate(A=@n8, B=TextureSample(Tex=@n9) {samplerType=SAMPLERTYPE_LinearColor}.R, Alpha=@n4.R)
   Normal=@n11.Result
   Refraction=LinearInterpolate(A=@n12, Alpha=@n4.R)
   Opacity=@n14
   EmissiveColor=LinearInterpolate(A=@n15, Alpha=@n4.R) {constB=0}
  ) {attributeSetTypes [
    #1
    #2
    #3
    #4
    #5
    #6
    #7
    #8
   ]}
  n17 = SetMaterialAttributes(BaseColor=@n1, Metallic=@n6.Result, Specular=@n7, Roughness=@n8, Normal=@n11.Result, Refraction=@n12, Opacity=@n14, EmissiveColor=@n15) {attributeSetTypes [
    #1
    #2
    #3
    #4
    #5
    #6
    #7
    #8
   ]}
  FunctionOutput(None=StaticSwitchParameter(True=@n16, False=@n17) {desc=Set to false to use as a standard glass material, group=00 - Global, parameterName=Use Frame Blend?}) {outputName=MA}
  FunctionOutput(None=@n14) {outputName=Opacity, sortPriority=1}
 ]}
