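# Material function in the Megascans glass material presets that computes an emissive value for translucent glass.
# It samples the Emissive Texture input as linear color, using the Tiling input as UVs, with a black texture as fallback.
# The RGB is scaled by the Max Emissive parameter times a constant hundredfold boost.
# The Use Emissive static switch selects that result or zero at compile time.
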MaterialFunction /Bridge/MSPresets/M_MS_Glass_Material/Functions/MF_TranslucencyEmission
graph {nodes [
  n1 = TextureSample(
   UVs=FunctionInput(Preview=Constant2Vector() {g=1, r=1}) {bUsePreviewValueAsDefault=true, inputName=Tiling, inputType=FunctionInput_Vector2}
   Tex=FunctionInput(Preview=TextureObject() {texture=&Black}) {bUsePreviewValueAsDefault=true, inputName=Emissive Texture, inputType=FunctionInput_Texture2D}
  ) {samplerType=SAMPLERTYPE_LinearColor}
  n2 = Multiply(A=ScalarParameter() {defaultValue=1, group=07 - Emissive, parameterName=Max Emissive, sliderMax=10, sortPriority=2}) {constB=100}
  FunctionOutput(None=StaticSwitchParameter(True=Multiply(A=@n1.RGB, B=@n2), False=Constant()) {group=07 - Emissive, parameterName=Use Emissive})
 ]}
