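# Material function for the glass material preset that outputs a window roughness value, switched between a constant base roughness and an imperfection texture.
# The texture input's red channel lerps between min and max parameters.
# Its UVs are tiled, offset and rotated by CustomRotator around the center of the tiled range.
# Only Tiling X and Offset X are wired, each applied to both axes, so the Y channels are unused.
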MaterialFunction /Bridge/MSPresets/M_MS_Glass_Material/Functions/MF_WindowImperfection
graph {nodes [
  n1 = ScalarParameter() {group=02 - Roughness, parameterName=Min Window Imperfection, sortPriority=2}
  n2 = ScalarParameter() {defaultValue=1, group=02 - Roughness, parameterName=Max Window Imperfection, sortPriority=3}
  n3 = VectorParameter() {
   group=02 - Roughness
   parameterName=Window Imperfection Tiling/Offset
   sortPriority=3
   channelNames {a=Offset Y, b=Offset X, g=Tiling Y, r=Tiling X}
   defaultValue=<rgba 1 1 0 0>
  }
  n4 = AppendVector(A=@n3.Tiling X, B=@n3.Tiling X)
  n5 = Multiply(A=TextureCoordinate(), B=@n4)
  n6 = AppendVector(A=@n3.Offset X, B=@n3.Offset X)
  n7 = Divide(A=@n4)
  n8 = TextureSample(
   UVs=MaterialFunctionCall(
    UVs=Add(A=@n5, B=@n6)
    Rotation Center=Add(A=@n7, B=@n6)
    "Rotation Angle (0-1)"=ScalarParameter() {group=02 - Roughness, parameterName=Window Imperfection Rotation, sortPriority=5}
   ) {materialFunction=&CustomRotator}.Rotated Values
   Tex=FunctionInput(Preview=TextureObject() {texture=/Bridge/MSPresets/MSTextures/BlackPlaceholder_Color}) {bUsePreviewValueAsDefault=true, inputName=Imperfection Texture, inputType=FunctionInput_Texture2D}
  )
  FunctionOutput(None=StaticSwitch(
    True=LinearInterpolate(A=@n1, B=@n2, Alpha=@n8.R)
    False=ScalarParameter() {defaultValue=0.01, group=02 - Roughness, parameterName=Base Window Roughness, sliderMax=1, sortPriority=1}
    Value=StaticBoolParameter() {group=02 - Roughness, parameterName=Use Window Imperfection Texture}
   ))
 ]}
