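# Megascans glass master material for translucent surfaces such as windows, using the Thin Translucent shading model with per-pixel surface lighting and index-of-refraction refraction.
# MF_FrameBlend builds the material attributes from glass color, specular, imperfection-driven roughness, texture-based emission and refraction.
# Static switches toggle the refraction function against a neutral value of 1, and choose transmittance color from the Albedo texture or the Color parameter.
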Material /Bridge/MSPresets/M_MS_Glass_Material/M_MS_Glass_Material
graph {
 properties {
  bUseMaterialAttributes=true
  bUsedWithStaticLighting=true
  blendMode=BLEND_Translucent
  pixelDepthOffsetMode=PDOM_Legacy
  refractionMethod=RM_IndexOfRefraction
  shadingModel=MSM_ThinTranslucent
  translucencyLightingMode=TLM_SurfacePerPixelLighting
  usedShadingModels=Thin Translucent
 }
 nodes [
  n1 = VectorParameter() {group=00 - Global, parameterName=Color, sortPriority=3, defaultValue=<rgba 1 1 1 1>}
  n2 = TextureObjectParameter() {
   group=07 - Texture Maps
   parameterName=Imperfection
   samplerType=SAMPLERTYPE_LinearColor
   sortPriority=4
   texture=/Bridge/MSPresets/MSTextures/BlackPlaceholder
  }
  n3 = TextureObjectParameter() {
   group=07 - Texture Maps
   parameterName=Emission
   samplerType=SAMPLERTYPE_LinearColor
   sortPriority=5
   texture=&BlackPlaceholder
  }
  MaterialFunctionCall!(
   Glass Color=@n1.RGB
   Roughness=MaterialFunctionCall!(Imperfection Texture=@n2) {materialFunction=/Bridge/MSPresets/M_MS_Glass_Material/Functions/MF_WindowImperfection}.Result
   Emissive=MaterialFunctionCall!(Tiling, Emissive Texture=@n3) {materialFunction=/Bridge/MSPresets/M_MS_Glass_Material/Functions/MF_TranslucencyEmission}.Result
   Specular=ScalarParameter() {defaultValue=0.5, group=00 - Global, parameterName=Specular, sortPriority=4}
   Refraction=StaticSwitchParameter(True=MaterialFunctionCall() {materialFunction=/Bridge/MSPresets/M_MS_Glass_Material/Functions/MF_Refraction}.Result, False=Constant() {r=1}) {defaultValue=true, group=05 - Refraction, parameterName=Use Refraction}
  ) {materialFunction=/Bridge/MSPresets/M_MS_Glass_Material/Functions/MF_FrameBlend}
  ThinTranslucentMaterialOutput!(
   TransmittanceColor=StaticSwitchParameter(
    True=TextureSampleParameter2D() {group=00 - Global, parameterName=Albedo, sortPriority=5, texture=/Bridge/MSPresets/MSTextures/Placeholder}.RGB
    False=@n1.RGB
   ) {defaultValue=true, group=00 - Global, parameterName=Use Texture for Transmittance Color, sortPriority=4}
   SurfaceCoverage
  )
 ]
}
